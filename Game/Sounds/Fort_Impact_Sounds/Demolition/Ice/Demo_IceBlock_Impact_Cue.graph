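# Sound cue for ice block impact sounds in the Demolition impact set.
# A Random node picks one of seven equally weighted ice impact waves, then passes it through a Modulator for variation.
# It plays in the impacts sound class with small melee weapon attenuation, and references an object-impact concurrency set to limit simultaneous instances.
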SoundCue /Game/Sounds/Fort_Impact_Sounds/Demolition/Ice/Demo_IceBlock_Impact_Cue
root=Modulator_37
nodes [
 {id=Modulator_37, children [Random_1]}
 {
  id=Random_1
  children [WavePlayer_4, WavePlayer_5, WavePlayer_6, WavePlayer_7, WavePlayer_8, WavePlayer_9, WavePlayer_10]
  properties {weights [1,1,1,1,1,1,1]}
 }
 {id=WavePlayer_10, properties {soundWaveAssetPtr=&/IceBlock_Impact_07}}
 {id=WavePlayer_4, properties {soundWaveAssetPtr=&/IceBlock_Impact_01}}
 {id=WavePlayer_5, properties {soundWaveAssetPtr=&/IceBlock_Impact_02}}
 {id=WavePlayer_6, properties {soundWaveAssetPtr=&/IceBlock_Impact_03}}
 {id=WavePlayer_7, properties {soundWaveAssetPtr=&/IceBlock_Impact_04}}
 {id=WavePlayer_8, properties {soundWaveAssetPtr=&/IceBlock_Impact_05}}
 {id=WavePlayer_9, properties {soundWaveAssetPtr=&/IceBlock_Impact_06}}
]
properties {
 attenuationSettings=&Fort_Weapon_Melee_Sml_Att
 maxDistance=2800
 soundClassObject=&Fort_Impacts
 volumeMultiplier=0.96
 concurrencyOverrides {maxCount=3, resolutionRule=StopOldest}
 concurrencySet [&CONC_Object_Impacts_FourMax]
}
